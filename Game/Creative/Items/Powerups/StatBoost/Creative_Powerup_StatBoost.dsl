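; Creative powerup Blueprint that supplies max health and max shield values to the gameplay effects it applies.
; Its SetupEffectData override calls the parent, then runs SetEffectDataForClass for health and shield in turn, threading the spec handle and should-apply flag through both.
; That helper assigns a SetByCaller magnitude on a matching effect class, blocks the effect if the magnitude is not positive, and passes other classes through unchanged.
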(event EventBeginPlay
  (|CallParentFunction) ; This node is disabled and will not be called.\nDrag off pins to build functionality.
)

(event Collision|EventActorBeginOverlap
  (params (OtherActor :type Actor%20Object%20Reference)))

(event EventTick
  (params (DeltaSeconds :type Float%20%28single-precision%29)))

(fn SetEffectDataForClass
  (params (GameplayEffectHandle :type Gameplay%20Effect%20Spec%20Handle%20Structure) (ClassToSetup :type Gameplay%20Effect%20Class%20Reference) (ClassToTest :type Gameplay%20Effect%20Class%20Reference) (CallerTag :type Gameplay%20Tag%20Structure) (Magnitude :type Float%20%28double-precision%29) (ShouldApply :type Boolean))
  (bind _localhandle (LocalVariables|GetLocalHandle))
  (LocalVariables|SetLocalHandle GameplayEffectHandle)
  (if (Utilities|Equal(Class) ClassToSetup ClassToTest)
    (if (> Magnitude 0.0)
      (bind _returnvalue (Ability|GameplayEffect|AssignTagSetbyCallerMagnitude (LocalVariables|GetLocalHandle) CallerTag Magnitude))
      (return true _returnvalue)
      (else
        (return false _localhandle)))
    (else
      ;[1 1 1] Not the testing effect so just pass back what was passed in
      (return ShouldApply _localhandle))))

(fn SetupEffectData
  (params (EffectHandle :type Gameplay%20Effect%20Spec%20Handle%20Structure) (EffectClass :type Gameplay%20Effect%20Class%20Reference))
  (bind _output_get (LocalVariables|SetCurrentClass EffectClass))
  (bind (_outeffecthandle _shouldapply) (|CallParentFunction EffectHandle _output_get))
  (LocalVariables|SetLocalHandle _outeffecthandle)
  (LocalVariables|SetLocalShouldApply _shouldapply)
  (Utilities|FlowControl|Sequence
    (:then_0
      (bind (_shouldapplyeffect _outgameplayefffecthandle) (CallFunction|SetEffectDataforClass (LocalVariables|GetLocalHandle) (LocalVariables|GetCurrentClass) (Variables|Default|GetMaxHealthClass) "(TagName=\"SetByCaller.MaxHealth\")" (Math|Conversions|ToFloat(Integer) (Variables|Default|GetMaxHealth)) (LocalVariables|GetLocalShouldApply)))
      (bind _output_get_2 (LocalVariables|SetLocalShouldApply _shouldapplyeffect))
      (if _output_get_2
        (LocalVariables|SetLocalHandle _outgameplayefffecthandle)))
    (:then_1
      (bind (_shouldapplyeffect_2 _outgameplayefffecthandle_2) (CallFunction|SetEffectDataforClass (LocalVariables|GetLocalHandle) (LocalVariables|GetCurrentClass) (Variables|Default|GetMaxShieldClass) "(TagName=\"SetByCaller.MaxShield\")" (Math|Conversions|ToFloat(Integer) (Variables|Default|GetMaxShield)) (LocalVariables|GetLocalShouldApply)))
      (bind _output_get_3 (LocalVariables|SetLocalShouldApply _shouldapplyeffect_2))
      (if _output_get_3
        (LocalVariables|SetLocalHandle _outgameplayefffecthandle_2)))
    (:then_2
      (return (LocalVariables|GetLocalHandle) (LocalVariables|GetLocalShouldApply)))))

(fn ConstructionScript
  (|CallParentFunction))
